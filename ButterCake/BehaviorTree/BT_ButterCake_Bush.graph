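# Behavior tree that drives ButterCake NPC bush behavior, luring it to a GoalActor bush and having it eat it.
# A root selector with a move-urgency override first tries eating: when ability and range checks pass, it faces the goal, waits briefly, then runs EatBush.
# Otherwise it faces the goal and runs the lured move-to-bush ability while applying a Lured status tag.
# Lower-priority aborts let eating interrupt the approach.
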BehaviorTree /ButterCake/BehaviorTree/BT_ButterCake_Bush
blackboard=&/BB_ButterCake_Base
root {
 name=Selector
 class=BTComposite_Selector
 children [
  Sequence {
   class=BTComposite_Sequence
   children [
    Rotate to face the Goal {
     class=FortBTTask_RotateToFaceBBEntryWithTags
     children []
     decorators []
     properties {
      bIgnoreRestartSelf=false
      nodeName=Rotate to face the Goal
      blackboardKey {bNoneIsAllowedValue=false, selectedKeyName=GoalActor}
      precision {defaultValue=10, key=None}
      tagsToApply {gameplayTags [{tagName=TurnTransition.Inplace.Allow}]}
     }
     services []
    }
    Wait {class=BTTask_Wait, properties {waitTime {defaultValue=0.5}}}
    EatBushAbility {
     class=FortBTTask_ExecuteGameplayAbility
     decorators []
     properties {nodeName=EatBushAbility, gameplayAbilityTagContainer {defaultValue {gameplayTags [{tagName=Ability.ButterCake.EatBush}]}}}
    }
   ]
   decorators [
    Has Gameplay Ability {
     class=FortBTDecorator_GameplayAbility_HasGameplayAbility
     properties {gameplayAbilityTagContainer {defaultValue {gameplayTags [{tagName=Ability.ButterCake.EatBush}]}}}
    }
    Can Activate Gameplay Ability {
     class=FortBTDecorator_GameplayAbility_CanActivate
     properties {
      flowAbortMode=LowerPriority
      gameplayAbilityTagContainer {defaultValue {gameplayTags [{tagName=Ability.ButterCake.EatBush}]}}
      targetValue {key=GoalActor}
     }
    }
    Close enough from Bush {
     class=FortBTDecorator_GameplayAbility_CompareDistance
     properties {
      bInverseCondition=false
      flowAbortMode=LowerPriority
      nodeName=Close enough from Bush
      activeAbilityTagsToSkipTestingValue {key=None, defaultValue {gameplayTags []}}
      distanceComparisons [{
        addedOffset=0
        bUseAddedOffset=false
        bUseOverriddenValue=false
        comparisonType=TwoDimensions
        operator=Less
        overriddenValue=0
        distanceDataTags {gameplayTags [{tagName=NPC.Ability.MaxActualRange}]}
       }]
      gameplayAbilityTagContainer {key=None, defaultValue {gameplayTags [{tagName=Ability.ButterCake.EatBush}]}}
      targetValue {baseClass=&Actor, defaultValue=null, key=GoalActor}
     }
    }
   ]
   properties {children [
     {childTask=&self:FortBTTask_RotateToFaceBBEntryWithTags_0}
     {childTask=&self:BTTask_Wait_0}
     {childTask=&self:FortBTTask_ExecuteGameplayAbility_8}
    ]}
  }
  Sequence {
   class=BTComposite_Sequence
   children [
    Rotate to face the Goal {
     class=FortBTTask_RotateToFaceBBEntryWithTags
     children []
     decorators []
     properties {
      bIgnoreRestartSelf=false
      nodeName=Rotate to face the Goal
      blackboardKey {bNoneIsAllowedValue=false, selectedKeyName=GoalActor}
      precision {defaultValue=10, key=None}
      tagsToApply {gameplayTags [{tagName=TurnTransition.Inplace.Allow}]}
     }
     services []
    }
    MoveToBushLuredAbility {
     class=FortBTTask_ExecuteGameplayAbility
     decorators []
     properties {nodeName=MoveToBushLuredAbility, gameplayAbilityTagContainer {defaultValue {gameplayTags [{tagName=Ability.ButterCake.MoveTo.Bush.Lured}]}}}
    }
   ]
   decorators [
    Has Gameplay Ability {
     class=FortBTDecorator_GameplayAbility_HasGameplayAbility
     properties {gameplayAbilityTagContainer {defaultValue {gameplayTags [{tagName=Ability.ButterCake.MoveTo.Bush.Lured}]}}}
    }
    Can Activate Gameplay Ability {
     class=FortBTDecorator_GameplayAbility_CanActivate
     properties {
      flowAbortMode=LowerPriority
      gameplayAbilityTagContainer {defaultValue {gameplayTags [{tagName=Ability.ButterCake.MoveTo.Bush.Lured}]}}
      targetValue {key=GoalActor}
     }
    }
   ]
   properties {
    children [
     {childTask=&self:FortBTTask_RotateToFaceBBEntryWithTags_1}
     {childTask=&self:FortBTTask_ExecuteGameplayAbility_0}
    ]
    services [&self:FortAthenaBTService_ApplyGameplayTags_0]
   }
   services [Apply Lured Tag {
     class=FortAthenaBTService_ApplyGameplayTags
     properties {interval=0.5, nodeName=Apply Lured Tag, gameplayTagsToApply {defaultValue {gameplayTags [{tagName=Athena.Gameplay.Status.Lured}]}}}
    }]
  }
 ]
 decorators []
 properties {
  children [
   {
    childComposite=&self:BTComposite_Sequence_3
    childTask=null
    decorators [
     &self:FortBTDecorator_GameplayAbility_HasGameplayAbility_0
     &self:FortBTDecorator_GameplayAbility_CanActivate_0
     &self:FortBTDecorator_GameplayAbility_CompareDistance_1
    ]
   }
   {
    childComposite=&self:BTComposite_Sequence_0
    childTask=null
    decorators [
     &self:FortBTDecorator_GameplayAbility_HasGameplayAbility_3
     &self:FortBTDecorator_GameplayAbility_CanActivate_2
    ]
   }
  ]
  services [&self:FortBTContext_MoveUrgency_0]
 }
 services [Move Urgency Override {
   class=FortBTContext_MoveUrgency
   properties {
    bCallTickOnSearchStart=false
    bRestartTimerOnEachActivation=false
    interval=0.5
    nodeName=Move Urgency Override
    randomDeviation=0.1
    enabled {defaultValue=true, key=None}
    moveUrgencyValue {defaultValue=2, enumType=/Script/FortniteGame.EFortMovementUrgency, key=None, nativeEnumTypeName=EFortMovementUrgency}
   }
  }]
}
