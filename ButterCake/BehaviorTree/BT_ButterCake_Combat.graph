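# Combat behavior tree for the ButterCake AI, choosing by priority among charge, ground pound, ranged attack, repositioning and a fallback wait.
# Gameplay ability checks gate each attack, which usually turns toward GoalActor before firing.
# Charge also requires a leashed, reachable goal; ground pound uses that check to strike in range or move closer, otherwise pounding anyway.
# Services on attack branches apply a tag preventing berserk from stopping, with movement urgency overrides for charge and pound.
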BehaviorTree /ButterCake/BehaviorTree/BT_ButterCake_Combat
blackboard=&/BB_ButterCake_Base
root {
 name=Selector
 class=BTComposite_Selector
 children [
  Sequence {
   class=BTComposite_Sequence
   children [
    Wait {class=BTTask_Wait, properties {waitTime {defaultValue=0.2}}}
    Rotate to face the Goal {
     class=FortBTTask_RotateToFaceBBEntryWithTags
     children []
     decorators []
     properties {
      bIgnoreRestartSelf=false
      nodeName=Rotate to face the Goal
      blackboardKey {bNoneIsAllowedValue=false, selectedKeyName=GoalActor}
      precision {defaultValue=10, key=None}
      tagsToApply {gameplayTags [{tagName=TurnTransition.Inplace.Allow}]}
     }
     services []
    }
    ChargeAttackAbility {
     class=FortBTTask_ExecuteGameplayAbility
     decorators []
     properties {nodeName=ChargeAttackAbility, gameplayAbilityTagContainer {defaultValue {gameplayTags [{tagName=Ability.ButterCake.Attack.Charge}]}}}
    }
   ]
   decorators [
    Goal Is Inside Leash ? {
     class=BTDecorator_Blackboard
     properties {
      arithmeticOperation=Equal
      intValue=0
      nodeName=Goal Is Inside Leash ?
      notifyObserver=ValueChange
      stringValue=""
      blackboardKey {selectedKeyName=AIEvaluator_Leash_GoalIsInsideLeash}
     }
    }
    Blackboard Based Condition {
     class=BTDecorator_Blackboard
     properties {
      arithmeticOperation=Equal
      intValue=9
      nodeName=Blackboard Based Condition
      stringValue=EBotMovementState::Success
      blackboardKey {selectedKeyName=AIEvaluator_PathExists}
     }
    }
    Blackboard Based Condition {
     class=BTDecorator_Blackboard
     properties {
      arithmeticOperation=Equal
      intValue=11
      nodeName=Blackboard Based Condition
      stringValue=EBotMovementState::Success_AlreadyAtGoal
      blackboardKey {selectedKeyName=AIEvaluator_PathExists}
     }
    }
    Has Gameplay Ability {
     class=FortBTDecorator_GameplayAbility_HasGameplayAbility
     properties {gameplayAbilityTagContainer {defaultValue {gameplayTags [{tagName=Ability.ButterCake.Attack.Charge}]}}}
    }
    Can Activate Gameplay Ability {
     class=FortBTDecorator_GameplayAbility_CanActivate
     properties {
      flowAbortMode=Both
      gameplayAbilityTagContainer {defaultValue {gameplayTags [{tagName=Ability.ButterCake.Attack.Charge}]}}
      targetValue {key=GoalActor}
     }
    }
   ]
   properties {
    children [
     {childTask=&self:BTTask_Wait_2}
     {childTask=&self:FortBTTask_RotateToFaceBBEntryWithTags_0}
     {childTask=&self:FortBTTask_ExecuteGameplayAbility_0}
    ]
    services [
     &self:FortBTContext_MoveUrgency_4
     &self:FortAthenaBTService_ApplyGameplayTags_4
    ]
   }
   services [
    Move Urgency Override {
     class=FortBTContext_MoveUrgency
     properties {
      bCallTickOnSearchStart=false
      bRestartTimerOnEachActivation=false
      interval=0.5
      nodeName=Move Urgency Override
      randomDeviation=0.1
      enabled {defaultValue=true, key=None}
      moveUrgencyValue {defaultValue=3, enumType=/Script/FortniteGame.EFortMovementUrgency, key=None, nativeEnumTypeName=EFortMovementUrgency}
     }
    }
    Prevent Stop Berserk {
     class=FortAthenaBTService_ApplyGameplayTags
     properties {
      interval=0.5
      nodeName=Prevent Stop Berserk
      gameplayTagsToApply {defaultValue {gameplayTags [{tagName=Athena.Gameplay.Status.PreventStopBerserk}]}}
     }
    }
   ]
  }
  Selector {
   class=BTComposite_Selector
   children [
    Selector {
     class=BTComposite_Selector
     children [
      Sequence {
       class=BTComposite_Sequence
       children [
        Rotate to face the Goal {
         class=FortBTTask_RotateToFaceBBEntryWithTags
         children []
         decorators []
         properties {
          bIgnoreRestartSelf=false
          nodeName=Rotate to face the Goal
          blackboardKey {bNoneIsAllowedValue=false, selectedKeyName=GoalActor}
          precision {defaultValue=10, key=None}
          tagsToApply {gameplayTags [{tagName=TurnTransition.Inplace.Allow}]}
         }
         services []
        }
        GroundPoundAbility {
         class=FortBTTask_ExecuteGameplayAbility
         decorators []
         properties {nodeName=GroundPoundAbility, gameplayAbilityTagContainer {defaultValue {gameplayTags [{tagName=Ability.ButterCake.Attack.GroundPound}]}}}
        }
       ]
       decorators [Close enough from Target to Attack {
         class=FortBTDecorator_GameplayAbility_CompareDistance
         properties {
          bInverseCondition=false
          flowAbortMode=LowerPriority
          nodeName=Close enough from Target to Attack
          activeAbilityTagsToSkipTestingValue {key=None, defaultValue {gameplayTags []}}
          distanceComparisons [{
            addedOffset=0
            bUseAddedOffset=false
            bUseOverriddenValue=false
            comparisonType=TwoDimensions
            operator=Less
            overriddenValue=0
            distanceDataTags {gameplayTags [{tagName=NPC.Ability.MaxActualRange}]}
           }]
          gameplayAbilityTagContainer {key=None, defaultValue {gameplayTags [{tagName=Ability.ButterCake.Attack.GroundPound}]}}
          targetValue {baseClass=&Actor, defaultValue=null, key=GoalActor}
         }
        }]
       properties {children [
         {childTask=&self:FortBTTask_RotateToFaceBBEntryWithTags_1}
         {childTask=&self:FortBTTask_ExecuteGameplayAbility_7}
        ]}
      }
      Sequence {
       class=BTComposite_Sequence
       children [
        Wait {class=BTTask_Wait, properties {waitTime {defaultValue=0.2}}}
        Rotate to face the Goal {
         class=FortBTTask_RotateToFaceBBEntryWithTags
         children []
         decorators []
         properties {
          bIgnoreRestartSelf=false
          nodeName=Rotate to face the Goal
          blackboardKey {bNoneIsAllowedValue=false, selectedKeyName=GoalActor}
          precision {defaultValue=10, key=None}
          tagsToApply {gameplayTags [{tagName=TurnTransition.Inplace.Allow}]}
         }
         services []
        }
        MoveToCombatTarget {
         class=FortBTTask_ExecuteGameplayAbility
         decorators []
         properties {nodeName=MoveToCombatTarget, gameplayAbilityTagContainer {defaultValue {gameplayTags [{tagName=Ability.ButterCake.MoveTo.CombatTarget}]}}}
        }
       ]
       decorators [
        Has Gameplay Ability {
         class=FortBTDecorator_GameplayAbility_HasGameplayAbility
         properties {gameplayAbilityTagContainer {defaultValue {gameplayTags [{tagName=Ability.ButterCake.MoveTo.CombatTarget}]}}}
        }
        Can Activate Gameplay Ability {
         class=FortBTDecorator_GameplayAbility_CanActivate
         properties {
          flowAbortMode=LowerPriority
          gameplayAbilityTagContainer {defaultValue {gameplayTags [{tagName=Ability.ButterCake.MoveTo.CombatTarget}]}}
          targetValue {key=GoalActor}
         }
        }
       ]
       properties {children [
         {childTask=&self:BTTask_Wait_4}
         {childTask=&self:FortBTTask_RotateToFaceBBEntryWithTags_2}
         {childTask=&self:FortBTTask_ExecuteGameplayAbility_1}
        ]}
      }
     ]
     decorators [
      Goal Is Inside Leash ? {
       class=BTDecorator_Blackboard
       properties {
        arithmeticOperation=Equal
        intValue=0
        nodeName=Goal Is Inside Leash ?
        notifyObserver=ValueChange
        stringValue=""
        blackboardKey {selectedKeyName=AIEvaluator_Leash_GoalIsInsideLeash}
       }
      }
      Blackboard Based Condition {
       class=BTDecorator_Blackboard
       properties {
        arithmeticOperation=Equal
        intValue=9
        nodeName=Blackboard Based Condition
        stringValue=EBotMovementState::Success
        blackboardKey {selectedKeyName=AIEvaluator_PathExists}
       }
      }
      Blackboard Based Condition {
       class=BTDecorator_Blackboard
       properties {
        arithmeticOperation=Equal
        intValue=11
        nodeName=Blackboard Based Condition
        stringValue=EBotMovementState::Success_AlreadyAtGoal
        blackboardKey {selectedKeyName=AIEvaluator_PathExists}
       }
      }
     ]
     properties {
      children [
       {
        childComposite=&self:BTComposite_Sequence_0
        childTask=null
        decorators [&self:FortBTDecorator_GameplayAbility_CompareDistance_3]
       }
       {
        childComposite=&self:BTComposite_Sequence_2
        childTask=null
        decorators [
         &self:FortBTDecorator_GameplayAbility_HasGameplayAbility_0
         &self:FortBTDecorator_GameplayAbility_CanActivate_0
        ]
       }
      ]
      services []
     }
     services []
    }
    GroundPound Random {
     class=FortBTTask_ExecuteGameplayAbility
     decorators []
     properties {nodeName=GroundPound Random, gameplayAbilityTagContainer {defaultValue {gameplayTags [{tagName=Ability.ButterCake.Attack.GroundPound}]}}}
    }
   ]
   decorators [
    Has Gameplay Ability {
     class=FortBTDecorator_GameplayAbility_HasGameplayAbility
     properties {gameplayAbilityTagContainer {defaultValue {gameplayTags [{tagName=Ability.ButterCake.Attack.GroundPound}]}}}
    }
    Can Activate Gameplay Ability {
     class=FortBTDecorator_GameplayAbility_CanActivate
     properties {
      flowAbortMode=LowerPriority
      gameplayAbilityTagContainer {defaultValue {gameplayTags [{tagName=Ability.ButterCake.Attack.GroundPound}]}}
      targetValue {key=GoalActor}
     }
    }
   ]
   properties {
    children [
     {
      childComposite=&self:BTComposite_Selector_6
      childTask=null
      decoratorOps [{number=2, operation=And}, {number=0}, {number=2, operation=Or}, {number=1}, {number=2}]
      decorators [
       &self:BTDecorator_Blackboard_0
       &self:BTDecorator_Blackboard_5
       &self:BTDecorator_Blackboard_6
      ]
     }
     {childComposite=null, childTask=&self:FortBTTask_ExecuteGameplayAbility_15, decorators []}
    ]
    services [
     &self:FortBTContext_MoveUrgency_2
     &self:FortAthenaBTService_ApplyGameplayTags_0
    ]
   }
   services [
    Move Urgency Override {
     class=FortBTContext_MoveUrgency
     properties {
      bCallTickOnSearchStart=false
      bRestartTimerOnEachActivation=false
      interval=0.5
      nodeName=Move Urgency Override
      randomDeviation=0.1
      enabled {defaultValue=true, key=None}
      moveUrgencyValue {defaultValue=2, enumType=&EFortMovementUrgency, key=None, nativeEnumTypeName=EFortMovementUrgency}
     }
    }
    Prevent Stop Berserk {
     class=FortAthenaBTService_ApplyGameplayTags
     properties {
      interval=0.5
      nodeName=Prevent Stop Berserk
      gameplayTagsToApply {defaultValue {gameplayTags [{tagName=Athena.Gameplay.Status.PreventStopBerserk}]}}
     }
    }
   ]
  }
  Sequence {
   class=BTComposite_Sequence
   children [
    Wait {class=BTTask_Wait, properties {waitTime {defaultValue=0.2}}}
    Rotate to face the Goal {
     class=FortBTTask_RotateToFaceBBEntryWithTags
     children []
     decorators []
     properties {
      bIgnoreRestartSelf=false
      nodeName=Rotate to face the Goal
      blackboardKey {bNoneIsAllowedValue=false, selectedKeyName=GoalActor}
      precision {defaultValue=10, key=None}
      tagsToApply {gameplayTags [{tagName=TurnTransition.Inplace.Allow}]}
     }
     services []
    }
    Range Attack Ability {
     class=FortBTTask_ExecuteGameplayAbility
     decorators []
     properties {nodeName=Range Attack Ability, gameplayAbilityTagContainer {defaultValue {gameplayTags [{tagName=Ability.ButterCake.Attack.Ranged}]}}}
    }
   ]
   decorators [
    Has Gameplay Ability {
     class=FortBTDecorator_GameplayAbility_HasGameplayAbility
     properties {gameplayAbilityTagContainer {defaultValue {gameplayTags [{tagName=Ability.ButterCake.Attack.Ranged}]}}}
    }
    Can Activate Gameplay Ability {
     class=FortBTDecorator_GameplayAbility_CanActivate
     properties {
      flowAbortMode=Both
      activeAbilityTagsToSkipTestingValue {defaultValue {gameplayTags [{tagName=Ability.ButterCake.Attack.Ranged}]}}
      gameplayAbilityTagContainer {defaultValue {gameplayTags [{tagName=Ability.ButterCake.Attack.Ranged}]}}
      targetValue {key=GoalActor}
     }
    }
   ]
   properties {
    children [
     {childTask=&self:BTTask_Wait_5}
     {childTask=&self:FortBTTask_RotateToFaceBBEntryWithTags_3}
     {childTask=&self:FortBTTask_ExecuteGameplayAbility_4}
    ]
    services [&self:FortAthenaBTService_ApplyGameplayTags_1]
   }
   services [Prevent Stop Berserk {
     class=FortAthenaBTService_ApplyGameplayTags
     properties {
      interval=0.5
      nodeName=Prevent Stop Berserk
      gameplayTagsToApply {defaultValue {gameplayTags [{tagName=Athena.Gameplay.Status.PreventStopBerserk}]}}
     }
    }]
  }
  Combat Repositioning Ability {
   class=FortBTTask_ExecuteGameplayAbility
   decorators [
    Has Gameplay Ability {
     class=FortBTDecorator_GameplayAbility_HasGameplayAbility
     properties {gameplayAbilityTagContainer {defaultValue {gameplayTags [{tagName=Ability.ButterCake.Attack.Repositioning}]}}}
    }
    Can Activate Gameplay Ability {
     class=FortBTDecorator_GameplayAbility_CanActivate
     properties {
      flowAbortMode=Both
      gameplayAbilityTagContainer {defaultValue {gameplayTags [{tagName=Ability.ButterCake.Attack.Repositioning}]}}
      targetValue {key=GoalActor}
     }
    }
   ]
   properties {
    nodeName=Combat Repositioning Ability
    gameplayAbilityTagContainer {defaultValue {gameplayTags [{tagName=Ability.ButterCake.Attack.Repositioning}]}}
   }
  }
  Wait - Combat Fallback {class=BTTask_Wait, properties {nodeName=Wait - Combat Fallback, waitTime {defaultValue=5}}}
 ]
 decorators []
 properties {
  children [
   {
    childComposite=&self:BTComposite_Sequence_4
    childTask=null
    decoratorOps [{number=4, operation=And}, {number=0}, {number=2, operation=Or}, {number=1}, {number=2}, {number=3}, {number=4}]
    decorators [
     &self:BTDecorator_Blackboard_12
     &self:BTDecorator_Blackboard_9
     &self:BTDecorator_Blackboard_10
     &self:FortBTDecorator_GameplayAbility_HasGameplayAbility_4
     &self:FortBTDecorator_GameplayAbility_CanActivate_4
    ]
   }
   {
    childComposite=&self:BTComposite_Selector_2
    childTask=null
    decorators [
     &self:FortBTDecorator_GameplayAbility_HasGameplayAbility_12
     &self:FortBTDecorator_GameplayAbility_CanActivate_11
    ]
   }
   {
    childComposite=&self:BTComposite_Sequence_1
    childTask=null
    decorators [
     &self:FortBTDecorator_GameplayAbility_HasGameplayAbility_2
     &self:FortBTDecorator_GameplayAbility_CanActivate_2
    ]
   }
   {
    childComposite=null
    childTask=&self:FortBTTask_ExecuteGameplayAbility_2
    decorators [
     &self:FortBTDecorator_GameplayAbility_HasGameplayAbility_1
     &self:FortBTDecorator_GameplayAbility_CanActivate_1
    ]
   }
   {childComposite=null, childTask=&self:BTTask_Wait_0, decorators []}
  ]
  services []
 }
 services []
}
