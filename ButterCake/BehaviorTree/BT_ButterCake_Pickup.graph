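# Behavior tree that drives the ButterCake NPC toward a GoalActor pickup and has it vacuum up loot.
# When either accepted path-evaluator result holds, it prioritizes vacuuming in range, then a lured move, then a normal move, each turning toward the goal first.
# The lured branch applies a Lured tag and move-urgency override, running EnterLuring first when possible.
# Gameplay ability checks gate each branch, with a loot-blacklist ability as fallback.
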BehaviorTree /ButterCake/BehaviorTree/BT_ButterCake_Pickup
blackboard=&/BB_ButterCake_Base
root {
 name=Selector
 class=BTComposite_Selector
 children [
  Selector {
   class=BTComposite_Selector
   children [
    Sequence {
     class=BTComposite_Sequence
     children [
      Rotate to face the Goal {
       class=FortBTTask_RotateToFaceBBEntryWithTags
       children []
       decorators []
       properties {
        bIgnoreRestartSelf=false
        nodeName=Rotate to face the Goal
        blackboardKey {bNoneIsAllowedValue=false, selectedKeyName=GoalActor}
        precision {defaultValue=10, key=None}
        tagsToApply {gameplayTags [{tagName=TurnTransition.Inplace.Allow}]}
       }
       services []
      }
      VaccuumPickupsAbility {
       class=FortBTTask_ExecuteGameplayAbility
       decorators []
       properties {nodeName=VaccuumPickupsAbility, gameplayAbilityTagContainer {defaultValue {gameplayTags [{tagName=Ability.ButterCake.VacuumPickups}]}}}
      }
     ]
     decorators [
      Has Gameplay Ability {
       class=FortBTDecorator_GameplayAbility_HasGameplayAbility
       properties {gameplayAbilityTagContainer {defaultValue {gameplayTags [{tagName=Ability.ButterCake.VacuumPickups}]}}}
      }
      Can Activate Gameplay Ability {
       class=FortBTDecorator_GameplayAbility_CanActivate
       properties {
        flowAbortMode=LowerPriority
        gameplayAbilityTagContainer {defaultValue {gameplayTags [{tagName=Ability.ButterCake.VacuumPickups}]}}
        targetValue {key=GoalActor}
       }
      }
      Close enough from Pickup {
       class=FortBTDecorator_GameplayAbility_CompareDistance
       properties {
        bInverseCondition=false
        flowAbortMode=LowerPriority
        nodeName=Close enough from Pickup
        activeAbilityTagsToSkipTestingValue {key=None, defaultValue {gameplayTags []}}
        distanceComparisons [{
          addedOffset=0
          bUseAddedOffset=false
          bUseOverriddenValue=false
          comparisonType=ThreeDimensions
          operator=Less
          overriddenValue=0
          distanceDataTags {gameplayTags [{tagName=NPC.Ability.MaxActualRange}]}
         }]
        gameplayAbilityTagContainer {key=None, defaultValue {gameplayTags [{tagName=Ability.ButterCake.VacuumPickups}]}}
        targetValue {baseClass=&Actor, defaultValue=null, key=GoalActor}
       }
      }
     ]
     properties {children [
       {childTask=&self:FortBTTask_RotateToFaceBBEntryWithTags_0}
       {childTask=&self:FortBTTask_ExecuteGameplayAbility_2}
      ]}
    }
    Selector {
     class=BTComposite_Selector
     children [
      Sequence {
       class=BTComposite_Sequence
       children [
        Rotate to face the Goal {
         class=FortBTTask_RotateToFaceBBEntryWithTags
         children []
         decorators []
         properties {
          bIgnoreRestartSelf=false
          nodeName=Rotate to face the Goal
          blackboardKey {bNoneIsAllowedValue=false, selectedKeyName=GoalActor}
          precision {defaultValue=3, key=None}
          tagsToApply {gameplayTags [{tagName=TurnTransition.Inplace.Allow}]}
         }
         services []
        }
        EnterLuring {
         class=FortBTTask_ExecuteGameplayAbility
         decorators []
         properties {nodeName=EnterLuring, gameplayAbilityTagContainer {defaultValue {gameplayTags [{tagName=Ability.ButterCake.EnterLuring}]}}}
        }
        MoveToPickupLuredAbility {
         class=FortBTTask_ExecuteGameplayAbility
         decorators []
         properties {
          nodeName=MoveToPickupLuredAbility
          gameplayAbilityTagContainer {defaultValue {gameplayTags [{tagName=Ability.ButterCake.MoveTo.Pickup.Lured}]}}
         }
        }
       ]
       decorators [
        Has Gameplay Ability {
         class=FortBTDecorator_GameplayAbility_HasGameplayAbility
         properties {gameplayAbilityTagContainer {defaultValue {gameplayTags [{tagName=Ability.ButterCake.EnterLuring}]}}}
        }
        Can Activate Gameplay Ability {
         class=FortBTDecorator_GameplayAbility_CanActivate
         properties {
          flowAbortMode=None
          gameplayAbilityTagContainer {defaultValue {gameplayTags [{tagName=Ability.ButterCake.EnterLuring}]}}
          targetValue {key=GoalActor}
         }
        }
       ]
       properties {children [
         {childTask=&self:FortBTTask_RotateToFaceBBEntryWithTags_1}
         {childTask=&self:FortBTTask_ExecuteGameplayAbility_13}
         {childTask=&self:FortBTTask_ExecuteGameplayAbility_6}
        ]}
      }
      Sequence {
       class=BTComposite_Sequence
       children [
        Rotate to face the Goal {
         class=FortBTTask_RotateToFaceBBEntryWithTags
         children []
         decorators []
         properties {
          bIgnoreRestartSelf=false
          nodeName=Rotate to face the Goal
          blackboardKey {bNoneIsAllowedValue=false, selectedKeyName=GoalActor}
          precision {defaultValue=3, key=None}
          tagsToApply {gameplayTags [{tagName=TurnTransition.Inplace.Allow}]}
         }
         services []
        }
        MoveToPickupLuredAbility {
         class=FortBTTask_ExecuteGameplayAbility
         decorators []
         properties {
          nodeName=MoveToPickupLuredAbility
          gameplayAbilityTagContainer {defaultValue {gameplayTags [{tagName=Ability.ButterCake.MoveTo.Pickup.Lured}]}}
         }
        }
       ]
       decorators []
       properties {children [
         {childTask=&self:FortBTTask_RotateToFaceBBEntryWithTags_2}
         {childTask=&self:FortBTTask_ExecuteGameplayAbility_16}
        ]}
      }
     ]
     decorators [
      Has Gameplay Ability {
       class=FortBTDecorator_GameplayAbility_HasGameplayAbility
       properties {gameplayAbilityTagContainer {defaultValue {gameplayTags [{tagName=Ability.ButterCake.MoveTo.Pickup.Lured}]}}}
      }
      Can Activate Gameplay Ability {
       class=FortBTDecorator_GameplayAbility_CanActivate
       properties {
        flowAbortMode=LowerPriority
        gameplayAbilityTagContainer {defaultValue {gameplayTags [{tagName=Ability.ButterCake.MoveTo.Pickup.Lured}]}}
        targetValue {key=GoalActor}
       }
      }
     ]
     properties {
      children [
       {
        childComposite=&self:BTComposite_Sequence_6
        childTask=null
        decorators [
         &self:FortBTDecorator_GameplayAbility_HasGameplayAbility_13
         &self:FortBTDecorator_GameplayAbility_CanActivate_13
        ]
       }
       {childComposite=&self:BTComposite_Sequence_7, childTask=null, decorators []}
      ]
      services [
       &self:FortBTContext_MoveUrgency_3
       &self:FortAthenaBTService_ApplyGameplayTags_2
      ]
     }
     services [
      Move Urgency Override {
       class=FortBTContext_MoveUrgency
       properties {
        bCallTickOnSearchStart=false
        bRestartTimerOnEachActivation=false
        interval=0.5
        nodeName=Move Urgency Override
        randomDeviation=0.1
        enabled {defaultValue=true, key=None}
        moveUrgencyValue {defaultValue=2, enumType=/Script/FortniteGame.EFortMovementUrgency, key=None, nativeEnumTypeName=EFortMovementUrgency}
       }
      }
      Apply Lured Tag {
       class=FortAthenaBTService_ApplyGameplayTags
       properties {interval=0.5, nodeName=Apply Lured Tag, gameplayTagsToApply {defaultValue {gameplayTags [{tagName=Athena.Gameplay.Status.Lured}]}}}
      }
     ]
    }
    Sequence {
     class=BTComposite_Sequence
     children [
      Rotate to face the Goal {
       class=FortBTTask_RotateToFaceBBEntryWithTags
       children []
       decorators []
       properties {
        bIgnoreRestartSelf=false
        nodeName=Rotate to face the Goal
        blackboardKey {bNoneIsAllowedValue=false, selectedKeyName=GoalActor}
        precision {defaultValue=10, key=None}
        tagsToApply {gameplayTags [{tagName=TurnTransition.Inplace.Allow}]}
       }
       services []
      }
      MoveToNormalPickupAbility {
       class=FortBTTask_ExecuteGameplayAbility
       decorators []
       properties {
        nodeName=MoveToNormalPickupAbility
        gameplayAbilityTagContainer {defaultValue {gameplayTags [{tagName=NPC.Ability.NonAttack.Move.ToPickup}]}}
       }
      }
     ]
     decorators [
      Has Gameplay Ability {
       class=FortBTDecorator_GameplayAbility_HasGameplayAbility
       properties {gameplayAbilityTagContainer {defaultValue {gameplayTags [{tagName=NPC.Ability.NonAttack.Move.ToPickup}]}}}
      }
      Can Activate Gameplay Ability {
       class=FortBTDecorator_GameplayAbility_CanActivate
       properties {
        flowAbortMode=LowerPriority
        gameplayAbilityTagContainer {defaultValue {gameplayTags [{tagName=NPC.Ability.NonAttack.Move.ToPickup}]}}
        targetValue {key=GoalActor}
       }
      }
     ]
     properties {children [
       {childTask=&self:FortBTTask_RotateToFaceBBEntryWithTags_3}
       {childTask=&self:FortBTTask_ExecuteGameplayAbility_9}
      ]}
    }
   ]
   decorators [
    Blackboard Based Condition {
     class=BTDecorator_Blackboard
     properties {
      arithmeticOperation=Equal
      intValue=9
      nodeName=Blackboard Based Condition
      stringValue=""
      blackboardKey {selectedKeyName=AIEvaluator_PathExists}
     }
    }
    Blackboard Based Condition {
     class=BTDecorator_Blackboard
     properties {
      arithmeticOperation=Equal
      intValue=11
      nodeName=Blackboard Based Condition
      stringValue=""
      blackboardKey {selectedKeyName=AIEvaluator_PathExists}
     }
    }
   ]
   properties {
    children [
     {
      childComposite=&self:BTComposite_Sequence_2
      childTask=null
      decorators [
       &self:FortBTDecorator_GameplayAbility_HasGameplayAbility_0
       &self:FortBTDecorator_GameplayAbility_CanActivate_0
       &self:FortBTDecorator_GameplayAbility_CompareDistance_1
      ]
     }
     {
      childComposite=&self:BTComposite_Selector_5
      childTask=null
      decorators [
       &self:FortBTDecorator_GameplayAbility_HasGameplayAbility_15
       &self:FortBTDecorator_GameplayAbility_CanActivate_15
      ]
     }
     {
      childComposite=&self:BTComposite_Sequence_8
      childTask=null
      decorators [
       &self:FortBTDecorator_GameplayAbility_HasGameplayAbility_1
       &self:FortBTDecorator_GameplayAbility_CanActivate_5
      ]
     }
    ]
    services []
   }
   services []
  }
  Black List Loot Ability {
   class=FortBTTask_ExecuteGameplayAbility
   decorators [
    Has Gameplay Ability {
     class=FortBTDecorator_GameplayAbility_HasGameplayAbility
     properties {gameplayAbilityTagContainer {defaultValue {gameplayTags [{tagName=Ability.ButterCake.BlackListLoot}]}}}
    }
    Can Activate Gameplay Ability {
     class=FortBTDecorator_GameplayAbility_CanActivate
     properties {
      flowAbortMode=LowerPriority
      gameplayAbilityTagContainer {defaultValue {gameplayTags [{tagName=Ability.ButterCake.BlackListLoot}]}}
      targetValue {key=GoalActor}
     }
    }
   ]
   properties {nodeName=Black List Loot Ability, gameplayAbilityTagContainer {defaultValue {gameplayTags [{tagName=Ability.ButterCake.BlackListLoot}]}}}
  }
 ]
 decorators []
 properties {
  children [
   {
    childComposite=&self:BTComposite_Selector_1
    childTask=null
    decoratorOps [{number=2, operation=Or}, {number=0}, {number=1}]
    decorators [
     &self:BTDecorator_Blackboard_0
     &self:BTDecorator_Blackboard_1
    ]
   }
   {
    childComposite=null
    childTask=&self:FortBTTask_ExecuteGameplayAbility_5
    decorators [
     &self:FortBTDecorator_GameplayAbility_HasGameplayAbility_3
     &self:FortBTDecorator_GameplayAbility_CanActivate_2
    ]
   }
  ]
  services []
 }
 services []
}
