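# Main behavior tree for the ButterCake NPC, a priority selector over dormancy, leash return, sneeze, goal handling, dance, idle, wander and fallback wait.
# AI evaluator services write dormancy, leash, nearby-actor and goal-path results to the blackboard for branch decorators.
# GoalActor selects Combat, Bush or Pickup subtrees by class and conditions.
# Berserk tag enables Combat and blocks sneeze and dance, while ability checks gate most action branches.
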BehaviorTree /ButterCake/BehaviorTree/BT_ButterCake
blackboard=&/BB_ButterCake_Base
root {
 name=Selector
 class=BTComposite_Selector
 children [
  Dormant_Wait4ever {
   class=BTTask_Wait
   decorators [Blackboard Based Condition {
     class=BTDecorator_Blackboard
     properties {
      arithmeticOperation=GreaterOrEqual
      flowAbortMode=Self
      intValue=3
      nodeName=Blackboard Based Condition
      stringValue=EExecutionStatus::ExecutionPending
      blackboardKey {selectedKeyName=AIEvaluator_Dormant_ExecutionStatus}
     }
    }]
   properties {nodeName=Dormant_Wait4ever, waitTime {defaultValue=10000000000}}
  }
  Selector {
   class=BTComposite_Selector
   children [
    Back to leash Ability {
     class=FortBTTask_ExecuteGameplayAbility
     decorators [IsBerserk? {
       class=BTDecorator_GameplayTagQuery
       properties {
        flowAbortMode=Both
        nodeName=IsBerserk?
        gameplayTagQuery {
         autoDescription=" ANY( Athena.Gameplay.Status.Berserk )"
         queryTokenStream [0,1,1,1,0]
         tagDictionary [{tagName=Athena.Gameplay.Status.Berserk}]
        }
       }
      }]
     properties {nodeName=Back to leash Ability, gameplayAbilityTagContainer {defaultValue {gameplayTags [{tagName=Ability.ButterCake.BackToLeash}]}}}
    }
    Back to leash Ability {
     class=FortBTTask_ExecuteGameplayAbility
     decorators []
     properties {nodeName=Back to leash Ability, gameplayAbilityTagContainer {defaultValue {gameplayTags [{tagName=Ability.ButterCake.BackToLeash}]}}}
    }
   ]
   decorators [
    Is outside the leash ? {
     class=BTDecorator_Blackboard
     properties {
      arithmeticOperation=Equal
      basicOperation=NotSet
      intValue=0
      nodeName=Is outside the leash ?
      notifyObserver=ValueChange
      stringValue=""
      blackboardKey {selectedKeyName=AIEvaluator_Leash_IsInsideLeash}
     }
    }
    Has Gameplay Ability {
     class=FortBTDecorator_GameplayAbility_HasGameplayAbility
     properties {gameplayAbilityTagContainer {defaultValue {gameplayTags [{tagName=Ability.ButterCake.BackToLeash}]}}}
    }
    Can Activate Gameplay Ability {
     class=FortBTDecorator_GameplayAbility_CanActivate
     properties {
      flowAbortMode=Both
      gameplayAbilityTagContainer {defaultValue {gameplayTags [{tagName=Ability.ButterCake.BackToLeash}]}}
      targetValue {key=GoalActor}
     }
    }
   ]
   properties {
    children [
     {
      childComposite=null
      childTask=&self:FortBTTask_ExecuteGameplayAbility_1
      decorators [&self:BTDecorator_GameplayTagQuery_1]
     }
     {childComposite=null, childTask=&self:FortBTTask_ExecuteGameplayAbility_4, decorators []}
    ]
    services []
   }
   services []
  }
  SneezAbility {
   class=FortBTTask_ExecuteGameplayAbility
   decorators [
    Has Gameplay Ability {
     class=FortBTDecorator_GameplayAbility_HasGameplayAbility
     properties {gameplayAbilityTagContainer {defaultValue {gameplayTags [{tagName=Ability.ButterCake.Sneeze}]}}}
    }
    Can Activate Gameplay Ability {
     class=FortBTDecorator_GameplayAbility_CanActivate
     properties {
      flowAbortMode=LowerPriority
      gameplayAbilityTagContainer {defaultValue {gameplayTags [{tagName=Ability.ButterCake.Sneeze}]}}
      targetValue {key=GoalActor}
     }
    }
    Is Not Berserk? {
     class=BTDecorator_GameplayTagQuery
     properties {
      flowAbortMode=Both
      nodeName=Is Not Berserk?
      gameplayTagQuery {
       autoDescription=" NONE( Athena.Gameplay.Status.Berserk )"
       queryTokenStream [0,1,3,1,0]
       tagDictionary [{tagName=Athena.Gameplay.Status.Berserk}]
      }
     }
    }
   ]
   properties {nodeName=SneezAbility, gameplayAbilityTagContainer {defaultValue {gameplayTags [{tagName=Ability.ButterCake.Sneeze}]}}}
  }
  Selector {
   class=BTComposite_Selector
   children [
    SBT Combat {
     class=BTTask_RunBehavior
     decorators [
      Goal is a Pawn {
       class=FortBTDecorator_IsGoalPawn
       properties {bInverseCondition=false, flowAbortMode=Both, nodeName=Goal is a Pawn, blackboardKey {bNoneIsAllowedValue=false, selectedKeyName=GoalActor}}
      }
      IsBerserk? {
       class=BTDecorator_GameplayTagQuery
       properties {
        flowAbortMode=Both
        nodeName=IsBerserk?
        gameplayTagQuery {
         autoDescription=" ANY( Athena.Gameplay.Status.Berserk )"
         queryTokenStream [0,1,1,1,0]
         tagDictionary [{tagName=Athena.Gameplay.Status.Berserk}]
        }
       }
      }
     ]
     properties {behaviorAsset=&/BT_ButterCake_Combat, nodeName=SBT Combat}
    }
    SBT Bush {
     class=BTTask_RunBehavior
     decorators [
      Is BlackBoard value of given Class {
       class=BTDecorator_IsBBEntryOfClass
       properties {
        bInverseCondition=false
        flowAbortMode=Both
        nodeName=Is BlackBoard value of given Class
        blackboardKey {bNoneIsAllowedValue=false, selectedKeyName=GoalActor}
        testClass {baseClass=&Object, defaultValue=/ButterCake/ButterCakeFruit/BGA_ButterCakeFruitPlant.BGA_ButterCakeFruitPlant_C, key=None}
       }
      }
      Blackboard Based Condition {
       class=BTDecorator_Blackboard
       properties {
        arithmeticOperation=Equal
        intValue=9
        nodeName=Blackboard Based Condition
        stringValue=EBotMovementState::Success
        blackboardKey {selectedKeyName=AIEvaluator_PathExists}
       }
      }
      Blackboard Based Condition {
       class=BTDecorator_Blackboard
       properties {
        arithmeticOperation=Equal
        intValue=11
        nodeName=Blackboard Based Condition
        stringValue=EBotMovementState::Success_AlreadyAtGoal
        blackboardKey {selectedKeyName=AIEvaluator_PathExists}
       }
      }
      Goal Is Inside Leash ? {
       class=BTDecorator_Blackboard
       properties {
        arithmeticOperation=Equal
        intValue=0
        nodeName=Goal Is Inside Leash ?
        notifyObserver=ValueChange
        stringValue=""
        blackboardKey {selectedKeyName=AIEvaluator_Leash_GoalIsInsideLeash}
       }
      }
     ]
     properties {behaviorAsset=&/BT_ButterCake_Bush, nodeName=SBT Bush}
    }
    SBT Pickup {
     class=BTTask_RunBehavior
     decorators [
      Is BlackBoard value of given Class {
       class=BTDecorator_IsBBEntryOfClass
       properties {
        bInverseCondition=false
        flowAbortMode=Both
        nodeName=Is BlackBoard value of given Class
        blackboardKey {bNoneIsAllowedValue=false, selectedKeyName=GoalActor}
        testClass {baseClass=&Object, defaultValue=/Script/FortniteGame.FortPickup, key=None}
       }
      }
      Is BlackBoard value of given Class {
       class=BTDecorator_IsBBEntryOfClass
       properties {
        bInverseCondition=false
        flowAbortMode=Both
        nodeName=Is BlackBoard value of given Class
        blackboardKey {bNoneIsAllowedValue=false, selectedKeyName=GoalActor}
        testClass {baseClass=&Object, defaultValue=/Script/FortniteGame.FortProjectileBase, key=None}
       }
      }
      Goal Is Inside Leash ? {
       class=BTDecorator_Blackboard
       properties {
        arithmeticOperation=Equal
        intValue=0
        nodeName=Goal Is Inside Leash ?
        notifyObserver=ValueChange
        stringValue=""
        blackboardKey {selectedKeyName=AIEvaluator_Leash_GoalIsInsideLeash}
       }
      }
     ]
     properties {behaviorAsset=&/BT_ButterCake_Pickup, nodeName=SBT Pickup}
    }
   ]
   decorators [Has GoalActor? {
     class=BTDecorator_Blackboard
     properties {
      arithmeticOperation=Equal
      intValue=0
      nodeName=Has GoalActor?
      notifyObserver=ValueChange
      stringValue=""
      blackboardKey {selectedKeyName=GoalActor}
     }
    }]
   properties {
    children [
     {
      childComposite=null
      childTask=&self:BTTask_RunBehavior_1
      decorators [&self:FortBTDecorator_IsGoalPawn_0, &self:BTDecorator_GameplayTagQuery_4]
     }
     {
      childComposite=null
      childTask=&self:BTTask_RunBehavior_2
      decoratorOps [{number=3, operation=And}, {number=0}, {number=2, operation=Or}, {number=1}, {number=2}, {number=3}]
      decorators [
       &self:BTDecorator_IsBBEntryOfClass_10
       &self:BTDecorator_Blackboard_19
       &self:BTDecorator_Blackboard_20
       &self:BTDecorator_Blackboard_0
      ]
     }
     {
      childComposite=null
      childTask=&self:BTTask_RunBehavior_0
      decoratorOps [{number=2, operation=And}, {number=2, operation=Or}, {number=0}, {number=1}, {number=2}]
      decorators [
       &self:BTDecorator_IsBBEntryOfClass_2
       &self:BTDecorator_IsBBEntryOfClass_3
       &self:BTDecorator_Blackboard_9
      ]
     }
    ]
    services [&self:FortAthenaBTService_AIEvaluator_1]
   }
   services [Evaluating...Path Exists {
     class=FortAthenaBTService_AIEvaluator
     properties {
      bCallTickOnSearchStart=true
      bRestartTimerOnEachActivation=true
      interval=0.75
      nodeName=Evaluating...Path Exists
      randomDeviation=0.25
      aIEvaluatorClass {defaultValue=/ButterCake/AIEvaluators/ButterCake_AIEval_PathExists.ButterCake_AIEval_PathExists_C}
     }
    }]
  }
  Selector {
   class=BTComposite_Selector
   children [
    Sequence {
     class=BTComposite_Sequence
     children [
      Wait {class=BTTask_Wait, properties {waitTime {defaultValue=1}}}
      Rotate to face the player {
       class=FortBTTask_RotateToFaceBBEntryWithTags
       children []
       decorators []
       properties {
        bIgnoreRestartSelf=false
        nodeName=Rotate to face the player
        blackboardKey {bNoneIsAllowedValue=false, selectedKeyName=AIEvaluator_FoundNearbyActor}
        precision {defaultValue=10, key=None}
        tagsToApply {gameplayTags [{tagName=TurnTransition.Inplace.Allow}]}
       }
       services []
      }
      DanceAbility {
       class=FortBTTask_ExecuteGameplayAbility
       decorators []
       properties {nodeName=DanceAbility, gameplayAbilityTagContainer {defaultValue {gameplayTags [{tagName=Ability.ButterCake.Dance}]}}}
      }
     ]
     decorators [Distance Between {
       class=FortBTDecorator_DistanceBetween
       properties {
        flowAbortMode=None
        nodeName=Distance Between
        operator=Greater
        bCalculateAs2D {defaultValue=true}
        blackboardKeyB {selectedKeyName=AIEvaluator_FoundNearbyActor}
        specifiedDistance {defaultValue=1500}
       }
      }]
     properties {children [
       {childTask=&self:BTTask_Wait_5}
       {childTask=&self:FortBTTask_RotateToFaceBBEntryWithTags_0}
       {childTask=&self:FortBTTask_ExecuteGameplayAbility_5}
      ]}
    }
    Sequence {
     class=BTComposite_Sequence
     children [
      Wait {class=BTTask_Wait, properties {waitTime {defaultValue=1}}}
      DanceAbility {
       class=FortBTTask_ExecuteGameplayAbility
       decorators []
       properties {nodeName=DanceAbility, gameplayAbilityTagContainer {defaultValue {gameplayTags [{tagName=Ability.ButterCake.Dance}]}}}
      }
     ]
     decorators []
     properties {children [
       {childTask=&self:BTTask_Wait_6}
       {childTask=&self:FortBTTask_ExecuteGameplayAbility_7}
      ]}
    }
   ]
   decorators [
    "Nearby Player? (Note :Abort must be Lower)" {
     class=BTDecorator_Blackboard
     properties {
      arithmeticOperation=GreaterOrEqual
      flowAbortMode=LowerPriority
      intValue=3
      nodeName="Nearby Player? (Note :Abort must be Lower)"
      stringValue=EExecutionStatus::ExecutionPending
      blackboardKey {selectedKeyName=AIEvaluator_FoundNearbyActor_ExecutionStatus}
     }
    }
    Has Gameplay Ability {
     class=FortBTDecorator_GameplayAbility_HasGameplayAbility
     properties {gameplayAbilityTagContainer {defaultValue {gameplayTags [{tagName=Ability.ButterCake.Dance}]}}}
    }
    Can Activate Gameplay Ability {
     class=FortBTDecorator_GameplayAbility_CanActivate
     properties {
      flowAbortMode=LowerPriority
      gameplayAbilityTagContainer {defaultValue {gameplayTags [{tagName=Ability.ButterCake.Dance}]}}
      targetValue {key=GoalActor}
     }
    }
    Is Not Berserk? {
     class=BTDecorator_GameplayTagQuery
     properties {
      flowAbortMode=Both
      nodeName=Is Not Berserk?
      gameplayTagQuery {
       autoDescription=" NONE( Athena.Gameplay.Status.Berserk )"
       queryTokenStream [0,1,3,1,0]
       tagDictionary [{tagName=Athena.Gameplay.Status.Berserk}]
      }
     }
    }
   ]
   properties {
    children [
     {
      childComposite=&self:BTComposite_Sequence_3
      childTask=null
      decorators [&self:FortBTDecorator_DistanceBetween_2]
     }
     {childComposite=&self:BTComposite_Sequence_4, childTask=null, decorators []}
    ]
    services []
   }
   services []
  }
  IdleAbility {
   class=FortBTTask_ExecuteGameplayAbility
   decorators [
    Has Gameplay Ability {
     class=FortBTDecorator_GameplayAbility_HasGameplayAbility
     properties {gameplayAbilityTagContainer {defaultValue {gameplayTags [{tagName=Ability.ButterCake.Idle}]}}}
    }
    Can Activate Gameplay Ability {
     class=FortBTDecorator_GameplayAbility_CanActivate
     properties {
      flowAbortMode=LowerPriority
      gameplayAbilityTagContainer {defaultValue {gameplayTags [{tagName=Ability.ButterCake.Idle}]}}
      targetValue {key=GoalActor}
     }
    }
   ]
   properties {nodeName=IdleAbility, gameplayAbilityTagContainer {defaultValue {gameplayTags [{tagName=Ability.ButterCake.Idle}]}}}
  }
  Selector {
   class=BTComposite_Selector
   children [
    Sequence {
     class=BTComposite_Sequence
     children [
      PatrolAroundAbility {
       class=FortBTTask_ExecuteGameplayAbility
       decorators []
       properties {nodeName=PatrolAroundAbility, gameplayAbilityTagContainer {defaultValue {gameplayTags [{tagName=NPC.Ability.NonAttack.Wander.Freely}]}}}
      }
      Wait {class=BTTask_Wait, properties {waitTime {defaultValue=3}}}
     ]
     decorators [IsBerserk? {
       class=BTDecorator_GameplayTagQuery
       properties {
        flowAbortMode=Both
        nodeName=IsBerserk?
        gameplayTagQuery {
         autoDescription=" ANY( Athena.Gameplay.Status.Berserk )"
         queryTokenStream [0,1,1,1,0]
         tagDictionary [{tagName=Athena.Gameplay.Status.Berserk}]
        }
       }
      }]
     properties {children [
       {childTask=&self:FortBTTask_ExecuteGameplayAbility_0}
       {childTask=&self:BTTask_Wait_3}
      ]}
    }
    Sequence {
     class=BTComposite_Sequence
     children [
      PatrolAroundAbility {
       class=FortBTTask_ExecuteGameplayAbility
       decorators []
       properties {nodeName=PatrolAroundAbility, gameplayAbilityTagContainer {defaultValue {gameplayTags [{tagName=NPC.Ability.NonAttack.Wander.Freely}]}}}
      }
      Wait {class=BTTask_Wait, properties {waitTime {defaultValue=3}}}
     ]
     decorators []
     properties {children [
       {childTask=&self:FortBTTask_ExecuteGameplayAbility_2}
       {childTask=&self:BTTask_Wait_2}
      ]}
    }
   ]
   decorators [
    Has Gameplay Ability {
     class=FortBTDecorator_GameplayAbility_HasGameplayAbility
     properties {gameplayAbilityTagContainer {defaultValue {gameplayTags [{tagName=NPC.Ability.NonAttack.Wander.Freely}]}}}
    }
    Can Activate Gameplay Ability {
     class=FortBTDecorator_GameplayAbility_CanActivate
     properties {
      flowAbortMode=LowerPriority
      gameplayAbilityTagContainer {defaultValue {gameplayTags [{tagName=NPC.Ability.NonAttack.Wander.Freely}]}}
      targetValue {key=GoalActor}
     }
    }
   ]
   properties {
    children [
     {
      childComposite=&self:BTComposite_Sequence_1
      childTask=null
      decorators [&self:BTDecorator_GameplayTagQuery_2]
     }
     {childComposite=&self:BTComposite_Sequence_0, childTask=null, decorators []}
    ]
    services []
   }
   services []
  }
  FallBack - Wait {class=BTTask_Wait, properties {nodeName=FallBack - Wait, waitTime {defaultValue=5}}}
 ]
 decorators []
 properties {
  children [
   {
    childComposite=null
    childTask=&self:BTTask_Wait_0
    decorators [&self:BTDecorator_Blackboard_2]
   }
   {
    childComposite=&self:BTComposite_Selector_0
    childTask=null
    decorators [
     &self:BTDecorator_Blackboard_6
     &self:FortBTDecorator_GameplayAbility_HasGameplayAbility_2
     &self:FortBTDecorator_GameplayAbility_CanActivate_4
    ]
   }
   {
    childComposite=null
    childTask=&self:FortBTTask_ExecuteGameplayAbility_3
    decorators [
     &self:FortBTDecorator_GameplayAbility_HasGameplayAbility_0
     &self:FortBTDecorator_GameplayAbility_CanActivate_0
     &self:BTDecorator_GameplayTagQuery_0
    ]
   }
   {
    childComposite=&self:BTComposite_Selector_3
    childTask=null
    decorators [&self:BTDecorator_Blackboard_1]
   }
   {
    childComposite=&self:BTComposite_Selector_5
    childTask=null
    decorators [
     &self:BTDecorator_Blackboard_3
     &self:FortBTDecorator_GameplayAbility_HasGameplayAbility_1
     &self:FortBTDecorator_GameplayAbility_CanActivate_1
     &self:BTDecorator_GameplayTagQuery_3
    ]
   }
   {
    childComposite=null
    childTask=&self:FortBTTask_ExecuteGameplayAbility_10
    decorators [
     &self:FortBTDecorator_GameplayAbility_HasGameplayAbility_3
     &self:FortBTDecorator_GameplayAbility_CanActivate_2
    ]
   }
   {
    childComposite=&self:BTComposite_Selector_1
    childTask=null
    decorators [
     &self:FortBTDecorator_GameplayAbility_HasGameplayAbility_5
     &self:FortBTDecorator_GameplayAbility_CanActivate_5
    ]
   }
   {childComposite=null, childTask=&self:BTTask_Wait_1, decorators []}
  ]
  services [
   &self:FortAthenaBTService_AIEvaluator_10
   &self:FortAthenaBTService_AIEvaluator_0
   &self:FortBTContext_MoveUrgency_6
   &self:FortAthenaBTService_AIEvaluator_2
  ]
 }
 services [
  Evaluating...DormantUntilPhase {
   class=FortAthenaBTService_AIEvaluator
   properties {
    interval=999999
    nodeName=Evaluating...DormantUntilPhase
    aIEvaluatorClass {defaultValue=/ButterCake/Evaluators/FortAthenaAIEvaluator_DormantUntilPhase_ButterCake.FortAthenaAIEvaluator_DormantUntilPhase_ButterCake_C}
   }
  }
  Evaluating... Leash {
   class=FortAthenaBTService_AIEvaluator
   properties {interval=0.5, nodeName=Evaluating... Leash, aIEvaluatorClass {defaultValue=/Script/FortniteAIServer.FortAthenaAIEvaluator_Leash}}
  }
  Move Urgency Override {
   class=FortBTContext_MoveUrgency
   properties {
    bCallTickOnSearchStart=false
    bRestartTimerOnEachActivation=false
    interval=0.5
    nodeName=Move Urgency Override
    randomDeviation=0.1
    enabled {defaultValue=true, key=None}
    moveUrgencyValue {defaultValue=1, enumType=/Script/FortniteGame.EFortMovementUrgency, key=None, nativeEnumTypeName=EFortMovementUrgency}
   }
  }
  Evaluating...NearbyDancingPlayer {
   class=FortAthenaBTService_AIEvaluator
   properties {
    interval=0.5
    nodeName=Evaluating...NearbyDancingPlayer
    aIEvaluatorClass {defaultValue=/ButterCake/AIEvaluators/ButterCake_AIEval_NearbyActorsPerception.ButterCake_AIEval_NearbyActorsPerception_C}
   }
  }
 ]
}
